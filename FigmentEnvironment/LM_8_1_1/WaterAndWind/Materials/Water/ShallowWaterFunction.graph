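# Material function that shades shallow water, outputting material attributes, their main channels and a player-distance falloff.
# Panning, distorted world-space normals warp a camera-parallax debris texture, which combines with vertex-red transition, Fresnel tint and falloff into base color and time-of-day emissive.
# A static switch optionally adds water interactions, while Forward and Mobile paths use flat color with layered panning normals.
# The pebble and mud blend feeds no output.
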MaterialFunction /FigmentEnvironment/LM_8_1_1/WaterAndWind/Materials/Water/ShallowWaterFunction
graph {nodes [
  n1 = CollectionParameter() {collection=&FortniteMaterialParameters, parameterName=PlayerPosition}
  n2 = Distance(A=ComponentMask(None=@n1) {b=true, g=true, r=true}, B=WorldPosition().XYZ)
  n3 = Clamp(None=Divide(A=Subtract(A=@n2) {constB=16}) {constB=64})
  FunctionOutput(None=@n3) {outputName=Distance falloff}
  n4 = VectorParameter() {group=Deep Water, parameterName=Ambient Deep Water Color, defaultValue=<rgba 0.0038 0.0081 0.015 1>}
  n5 = Constant() {desc=water is 0.4, r=0.4}
  n6 = ComponentMask(None=WorldPosition().XYZ) {g=true, r=true}
  n7 = Divide(A=@n6, B=ScalarParameter() {defaultValue=3934.5361, group=M, parameterName=UVSize_M_1})
  n8 = Time()
  n9 = Multiply(A=@n8, B=ScalarParameter() {defaultValue=0.05, group=M, parameterName=PanSpeed_M3})
  n10 = TextureSample(UVs=VertexInterpolator(VS=Panner(Coordinate=@n7, Time=@n9) {bFractionalPart=true, speedX=0.11, speedY=0.27}).PS) {samplerType=SAMPLERTYPE_Normal, texture=/FigmentEnvironment/LM_8_1_1/Packages/Fortnite_UniqueMaterials_1/Textures/T_WaterSoft_N}
  n11 = ScalarParameter() {defaultValue=0.9115, parameterName=NormalFlaten_M3}
  n12 = Divide(A=@n6, B=ScalarParameter() {defaultValue=1308.5537, group=M, parameterName=UVSize_M})
  n13 = Panner(Coordinate=@n12, Time=Multiply(A=@n8, B=ScalarParameter() {defaultValue=0.1, group=M, parameterName=PanSpeed_M})) {bFractionalPart=true, speedX=0.1, speedY=0.33}
  n14 = Panner(Coordinate=@n12, Time=Multiply(A=@n8, B=ScalarParameter() {defaultValue=0.05, group=M, parameterName=PanSpeed_M2})) {bFractionalPart=true, speedX=0.09, speedY=0.22}
  n15 = MaterialFunctionCall(
   BaseNormal=TextureSample(UVs=VertexInterpolator(VS=@n13).PS) {samplerType=SAMPLERTYPE_Normal, texture=/FigmentEnvironment/LM_8_1_1/Packages/Fortnite_UniqueMaterials_1/Textures/T_Water2_N}.RGB
   AdditionalNormal=TextureSample(UVs=VertexInterpolator(VS=@n14).PS) {samplerType=SAMPLERTYPE_Normal, texture=&T_Water2_N}.RGB
  ) {materialFunction=&BlendAngleCorrectedNormals}
  n16 = LinearInterpolate(
   A=ScalarParameter() {defaultValue=0.937, parameterName=NormalFlaten_M}
   B=ScalarParameter() {defaultValue=1, parameterName=NormalFlaten_M2}
   Alpha=MaterialFunctionCall(
    Fade Length=ScalarParameter() {defaultValue=5805.6235, parameterName=FadeLength_M}
    Fade Offset=ScalarParameter() {defaultValue=1429.6207, parameterName=FadeOffset_M}
   ) {materialFunction=&CameraDepthFade}.Result
  )
  n17 = MakeMaterialAttributes!(
   BaseColor=Multiply(A=@n4.RGB, B=ScalarParameter() {defaultValue=2, parameterName=Brightness_M})
   Metallic=Constant()
   Specular=@n5
   Roughness=Constant()
   Anisotropy
   EmissiveColor
   Opacity
   OpacityMask
   Normal=MaterialFunctionCall(
    BaseNormal=MaterialFunctionCall(Normal=@n10.RGB, Flatness=@n11) {materialFunction=&FlattenNormal}.Result
    AdditionalNormal=MaterialFunctionCall(Normal=@n15.Result, Flatness=@n16) {materialFunction=&FlattenNormal}.Result
   ) {materialFunction=&BlendAngleCorrectedNormals}.Result
   Tangent
   WorldPositionOffset
   SubsurfaceColor
   ClearCoat
   ClearCoatRoughness
   AmbientOcclusion
   Refraction
   CustomizedUVs_0
   CustomizedUVs_1
   CustomizedUVs_2
   CustomizedUVs_3
   CustomizedUVs_4
   CustomizedUVs_5
   CustomizedUVs_6
   CustomizedUVs_7
   PixelDepthOffset
   ShadingModel
   Displacement
  )
  n18 = CollectionParameter() {
   collection=&FortniteMaterialParameters
   parameterName=FortniteActiveTimeOfDay
  }
  n19 = VectorParameter() {group=Deep Water, parameterName=Time of day deep water emissive values, defaultValue=<rgba 0 0.6192 0 0>}
  n20 = AppendVector(A=@n19.RGB, B=@n19.RGB)
  n21 = VectorParameter() {group=Deep Water, parameterName=Deep Water Color 1, defaultValue=<rgba 0.271 0.6224 0.635 0>}
  n22 = VectorParameter() {group=Deep Water, parameterName=Deep Water Color 2, defaultValue=<rgba 0.5718 0.8684 1 0.3>}
  n23 = MaterialFunctionCall(Invert Fresnel=StaticBool() {value=true}, Power=ScalarParameter() {defaultValue=3, group=Deep Water, parameterName=POWER 33}) {materialFunction=&Fresnel_Function}
  n24 = Multiply(A=LinearInterpolate(A=@n21.RGB, B=@n22.RGB, Alpha=@n23.Result), B=LinearInterpolate(A=@n21.A, B=@n22.A, Alpha=@n23.Result))
  n25 = VertexColor()
  n26 = OneMinus(None=Multiply(A=@n25.R, B=@n25.R))
  n27 = ComponentMask(None=WorldPosition().XYZ) {g=true, r=true}
  n28 = Divide(A=@n27, B=ScalarParameter() {defaultValue=603.1323, parameterName=UVSize})
  n29 = Add(A=Multiply(A=@n28) {constB=0.8556}) {constB=0.3215}
  n30 = Add(A=@n28) {constB=0.1255}
  n31 = ComponentMask(None=TextureSample(UVs=Panner(Coordinate=@n30) {bFractionalPart=true, speedX=0.0208, speedY=-0.0278}) {samplerType=SAMPLERTYPE_Normal, texture=&T_Water2_N}.RGB) {g=true, r=true}
  n32 = TextureSample(UVs=Add(A=Panner(Coordinate=@n29) {bFractionalPart=true, speedX=0.0625, speedY=0.0417}, B=Multiply(A=@n31) {constB=0.035})) {samplerType=SAMPLERTYPE_Normal, texture=&T_Water2_N}
  n33 = ComponentMask(None=@n32.RGB) {g=true, r=true}
  n34 = ScalarParameter() {defaultValue=0.0442, group=Deep Water, parameterName=Noise 1}
  n35 = Panner(Coordinate=@n28) {bFractionalPart=true, speedX=-0.025, speedY=-0.0375}
  n36 = TextureSample(UVs=Add(A=Multiply(A=@n33, B=@n34) {constB=0.04}, B=Add(A=@n35) {constB=0.333})) {samplerType=SAMPLERTYPE_Normal, texture=&T_Water2_N}
  n37 = Abs(None=MaterialFunctionCall(f2=@n36.RGB) {materialFunction=&AddComponents}.f2)
  n38 = Normalize(VectorInput=MaterialFunctionCall(BaseNormal=@n32.RGB, AdditionalNormal=@n36.RGB) {materialFunction=&BlendAngleCorrectedNormals}.Result)
  n39 = Add(
   A=Divide(A=@n27, B=ScalarParameter() {defaultValue=530.4807, group=Deep Water, parameterName=Rock Tiling})
   B=Multiply(A=ComponentMask(None=@n38) {g=true, r=true}, B=ScalarParameter() {defaultValue=0.1, group=Deep Water, parameterName=Ground Distortion RRRR})
  )
  n40 = CameraVectorWS()
  n41 = FunctionInput() {bUsePreviewValueAsDefault=true, inputName=Height Offset Amount, inputType=FunctionInput_Scalar, previewValue=<1 1 0 1>}
  n42 = Multiply(A=ComponentMask(None=Normalize(VectorInput=@n40)) {g=true, r=true}, B=@n41)
  n43 = Multiply(A=@n42, B=Constant() {r=-0.15})
  n44 = Multiply(A=TextureSample(UVs=Add(A=@n39, B=@n43)) {desc=floating debris, texture=/FigmentEnvironment/LM_8_1_1/Environments/WorldMap/Textures/WM_StarField_D}.RGB) {constB=5}
  n45 = Constant()
  n46 = FeatureLevelSwitch(Default=@n44, ES2=@n45, ES3_1=@n45, SM4_REMOVED=@n44)
  n47 = Add(A=Multiply(A=Power(Base=@n26) {constExponent=5}, B=OneMinus(None=@n37)), B=@n46)
  n48 = FunctionInput() {bUsePreviewValueAsDefault=true, inputName=Ground texture mask, inputType=FunctionInput_Scalar, previewValue=<1 0 0 1>}
  n49 = Add(
   A=Multiply(A=StaticSwitchParameter(True=@n47, False=@n46) {defaultValue=true, parameterName=UseStreamLakeTransiton}, B=@n48)
   B=VectorParameter() {group=Deep Water, parameterName=BAse Sand color, defaultValue=<rgba 0.0119 0.0302 0.04 1>}.RGB
  )
  n50 = Add(A=Multiply(A=@n3, B=Multiply(A=@n24, B=@n49)), B=@n4.RGB)
  n51 = ScalarParameter() {defaultValue=0.89, parameterName=NormalFlatten}
  n52 = ScalarParameter() {defaultValue=0.99, parameterName=NormalFlatten_2}
  n53 = MaterialFunctionCall(Fade Length=Constant() {r=6000}, Fade Offset=Constant() {r=2000}) {materialFunction=&CameraDepthFade}
  n54 = MakeMaterialAttributes(
   BaseColor=@n50
   Specular=@n5
   Roughness=Constant()
   EmissiveColor=Multiply(A=@n50, B=DotProduct(A=@n18, B=@n20))
   Normal=MaterialFunctionCall(Normal=@n38, Flatness=LinearInterpolate(A=@n51, B=@n52, Alpha=@n53.Result)) {materialFunction=&FlattenNormal}.Result
  )
  n55 = MaterialFunctionCall!(Normal intensity, "World space offset to calculate normal ") {materialFunction=/FigmentEnvironment/LM_8_1_1/WaterAndWind/MaterialFunctions/Fortnite_SampleRealTimeWaterData}
  n56 = FunctionInput(Preview=@n55.Normal from height map) {bUsePreviewValueAsDefault=true, inputName=Sample Water Normal}
  n57 = FunctionInput(Preview=@n55.Height Map) {bUsePreviewValueAsDefault=true, inputName=Sample Water Height Map}
  n58 = FunctionInput(Preview=StaticBool()) {
   bUsePreviewValueAsDefault=true
   description=False by default
   inputName=Include water interaction modifications
   inputType=FunctionInput_StaticBool
  }
  n59 = ShadingPathSwitch(
   Default=StaticSwitch(
    True=MaterialFunctionCall!(In=@n54, Normal From Height Map=@n56, Height Map=@n57) {materialFunction=/FigmentEnvironment/LM_8_1_1/WaterAndWind/MaterialFunctions/Fortnite_WaterInteractions_BaseColorAndNormal}.Result
    False=@n54
    Value=@n58
   )
   Forward=@n17
   Mobile=@n17
  )
  FunctionOutput(None=@n59) {description=Base color specular roughness emissive and normal, outputName=MA Result}
  n60 = BreakMaterialAttributes!(Attr=@n59)
  FunctionOutput(None=@n60.Roughness) {description=Base color specular roughness emissive and normal, outputName=Roughness, sortPriority=3}
  FunctionOutput(None=@n60.EmissiveColor) {description=Base color specular roughness emissive and normal, outputName=Emissive, sortPriority=4}
  FunctionOutput(None=@n60.Normal) {description=Base color specular roughness emissive and normal, outputName=Normal, sortPriority=5}
  FunctionOutput(None=@n60.BaseColor) {description=Base color specular roughness emissive and normal, outputName=Base Color, sortPriority=1}
  FunctionOutput(None=@n60.Specular) {description=Base color specular roughness emissive and normal, outputName=Spec, sortPriority=2}
  OneMinus(None=Clamp(None=DotProduct(A=@n36.RGB, B=@n32.RGB)))
  n61 = FunctionInput() {bUsePreviewValueAsDefault=true, inputName=Texture mip bias amount, inputType=FunctionInput_Scalar, previewValue=<0 1 0 1>}
  n62 = Add(A=@n39, B=Multiply(A=Constant() {r=-0.3968}, B=@n42))
  n63 = TextureSample!(UVs=@n62, Tex, Bias=@n61, Apply View MipBias) {
   mipValueMode=TMVM_MipBias
   samplerSource=SSM_Wrap_WorldGroupSettings
   texture=/FigmentEnvironment/LM_8_1_1/Packages/DS_Fortnight/pebbles_blurred
  }
  n64 = TextureSample(UVs=@n62, Bias=Add(A=Constant() {r=2}, B=@n61)) {
   mipValueMode=TMVM_MipBias
   samplerSource=SSM_Wrap_WorldGroupSettings
   texture=/FigmentEnvironment/LM_8_1_1/Packages/DS_Fortnite_TRV_Terrain/Textures/T_Transylvania_HF_Mud_1_D
  }
  n65 = WorldPosition()
  n66 = TextureSample!(UVs=Divide(A=ComponentMask(None=@n65.XYZ) {g=true, r=true}) {constB=1024}, Tex, Apply View MipBias) {
   samplerSource=SSM_Wrap_WorldGroupSettings
   samplerType=SAMPLERTYPE_LinearGrayscale
   texture=/FigmentEnvironment/LM_8_1_1/Effects/Fort_Effects/BluGlow/Violentium/T_SwirlNoise_N_height
  }
  Reroute(None=LinearInterpolate(A=@n63.RGB, B=@n64.RGB, Alpha=@n66.R))
 ]}
