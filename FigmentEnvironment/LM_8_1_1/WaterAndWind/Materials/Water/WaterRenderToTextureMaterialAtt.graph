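# Material function that switches a water material between its normal look and a render-to-texture pass.
# A static switch, driven by the Render To Texture Pass input, selects between the supplied default material attributes and a capture-only set.
# The capture set sends the world-space flow map direction through SceneCaptureBlueChannel into Emissive, alongside an Opacity input.
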MaterialFunction /FigmentEnvironment/LM_8_1_1/WaterAndWind/Materials/Water/WaterRenderToTextureMaterialAtt
graph {nodes [
  n1 = StaticBoolParameter() {group=Render To Texture Pass, parameterName=Render To Texture Pass}
  n2 = MakeMaterialAttributes!(
   BaseColor
   Metallic
   Specular
   Roughness
   Anisotropy
   EmissiveColor
   Opacity
   OpacityMask
   Normal
   Tangent
   WorldPositionOffset
   SubsurfaceColor
   ClearCoat
   ClearCoatRoughness
   AmbientOcclusion
   Refraction
   CustomizedUVs_0
   CustomizedUVs_1
   CustomizedUVs_2
   CustomizedUVs_3
   CustomizedUVs_4
   CustomizedUVs_5
   CustomizedUVs_6
   CustomizedUVs_7
   PixelDepthOffset
   ShadingModel
   Displacement
  )
  n3 = MaterialFunctionCall!(WS Flow Map Dir=FunctionInput() {bUsePreviewValueAsDefault=true, inputName=WS Flow Map Dir, inputType=FunctionInput_Vector2}) {materialFunction=/FigmentEnvironment/LM_8_1_1/WaterAndWind/MaterialFunctions/SceneCaptureBlueChannel}
  n4 = FunctionInput() {bUsePreviewValueAsDefault=true, inputName=Opacity, inputType=FunctionInput_Scalar, sortPriority=1, previewValue=<1 0 0 1>}
  FunctionOutput(None=StaticSwitch(
    True=MakeMaterialAttributes(EmissiveColor=@n3.Result, Opacity=@n4)
    False=FunctionInput(Preview=@n2) {bUsePreviewValueAsDefault=true, inputName=Default Mat Att, inputType=FunctionInput_MaterialAttributes, sortPriority=10}
    Value=FunctionInput(Preview=@n1) {bUsePreviewValueAsDefault=true, inputName=Render To Texture Pass, inputType=FunctionInput_StaticBool, sortPriority=-1}
   ))
 ]}
